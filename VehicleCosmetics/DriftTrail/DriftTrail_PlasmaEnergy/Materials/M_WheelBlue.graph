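# Translucent unlit Niagara mesh-particle material for the Plasma Energy drift trail's wheel effect.
# Layers MF_SpherePlasma calls into two looks, chosen by the "Pink?" static switch, with dynamic parameters driving power and layer adjustments.
# When IsPainted? reaches 0.5, the layers are desaturated and tinted by particle color.
# The result scales by MultiColorFinal into MF_FX_EmissiveAdjustments; a second switch selects the pink result or summed plasma times PowerAlpha.
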Material /VehicleCosmetics/DriftTrail/DriftTrail_PlasmaEnergy/Materials/M_WheelBlue
graph {
 properties {
  bAutomaticallySetUsageInEditor=false
  bUsedWithNiagaraMeshParticles=true
  blendMode=BLEND_Translucent
  shadingModel=MSM_Unlit
  usedShadingModels=Unlit
 }
 nodes [
  Add()
  n1 = Constant2Vector() {r=0.1}
  n2 = Constant2Vector() {r=0.2}
  n3 = Constant3Vector() {constant=<rgba 1 1 1 0>}
  n4 = Constant() {r=2.5}
  n5 = Constant() {r=0.9}
  n6 = Constant() {r=3}
  n7 = Constant() {r=1}
  n8 = DynamicParameter() {defaultValue=<rgba 0 1 1 1>, paramNames [IsPainted?, PowerColorDefaultVersion, Param3, Param4]}
  n9 = Multiply(
   A=Saturate(None=MaterialFunctionCall!(
     2ndVelocity=@n1
     1stVelocity=@n2
     Color=@n3
     ColorFresnel=@n3
     PowerFrenelColor=@n4
     AdjustmenLayerPower=@n5
     Power_1stFresnel=@n6
     Alpha=@n7
    ) {materialFunction=/VehicleCosmetics/Boosts/Boost_PlasmaEnergy/Material/MF_SpherePlasma}."Alpha_{")
   B=Multiply(A=Constant3Vector() {constant=<rgba 0.5 0 1 0>}, B=@n8.PowerColorDefaultVersion)
  )
  n10 = Multiply(
   A=MaterialFunctionCall(
    2ndVelocity=Constant2Vector() {r=-0.6}
    1stVelocity=Constant2Vector() {r=-0.2}
    Color=@n3
    ColorFresnel=@n3
    PowerFrenelColor=@n4
    AdjustmenLayerPower=Constant() {r=0.8}
    Power_1stFresnel=Constant() {r=5}
    Alpha=Constant() {r=1}
   ) {materialFunction=&MF_SpherePlasma}."Alpha_{"
   B=Multiply(A=Constant3Vector() {constant=<rgba 0.7708 0.2313 0.7708 0>}, B=@n8.PowerColorDefaultVersion)
  )
  Add(A=@n9, B=@n10)
  Constant3Vector() {constant=<rgba 1 1 1 0>}
  Constant() {r=0.6}
  Constant() {r=1}
  Constant() {r=0.6}
  Constant() {r=0.5}
  Constant() {r=1}
  Constant() {r=0.6}
  Constant() {r=0.7}
  Constant() {r=1}
  Constant() {r=0.6}
  Constant() {r=1}
  Constant() {r=0.6}
  Constant() {r=1}
  Constant() {r=1}
  n11 = Multiply(A=LinearInterpolate(A=Desaturation(None=@n9), B=Desaturation(None=@n10)) {constAlpha=0.25}, B=ParticleColor().RGB)
  n12 = Multiply(
   A=If(
    A=DynamicParameter() {defaultValue=<rgba 0 1 1 1>, paramNames [IsPainted?, PowerColorDefaultVersion, Param3, Param4]}.IsPainted?
    "A > B"=@n11
    "A == B"=@n11
    "A < B"=LinearInterpolate(A=@n9, B=@n10) {constAlpha=0.25}
   ) {constB=0.5}
   B=ScalarParameter() {defaultValue=1, parameterName=MultiColorFinal}
  )
  n13 = DynamicParameter() {defaultValue=<rgba 0 1 1 1>, paramNames [IsPainted?, MultPowerPainted, Param3, Param4]}
  n14 = Constant3Vector() {constant=<rgba 1 1 1 0>}
  n15 = Constant3Vector() {constant=<rgba 0.1771 0.1771 0.1771 0>}
  n16 = Constant() {r=1}
  n17 = DynamicParameter() {parameterIndex=3, defaultValue=<rgba 0 1 1 1>, paramNames [AdjustmentLayer2, AdjustmentLayer1, "", ""]}
  n18 = Multiply(
   A=MaterialFunctionCall(
    Color=@n14
    ColorFresnel=@n15
    PowerFrenelColor=@n16
    AdjustmenLayerPower=Constant() {r=2}
    Power_1stFresnel=Constant() {r=3}
    Alpha=Constant() {r=1}
   ) {materialFunction=&MF_SpherePlasma}."Alpha_{"
   B=Multiply(A=@n17.A, B=@n17.A)
  )
  n19 = Desaturation(None=@n18)
  n20 = DynamicParameter() {parameterIndex=3, defaultValue=<rgba 0 1 1 1>, paramNames [AdjustmentLayer2, AdjustmentLayer1, "", ""]}
  n21 = MaterialFunctionCall(
   Color=@n14
   ColorFresnel=@n15
   PowerFrenelColor=@n16
   AdjustmenLayerPower=@n20.AdjustmentLayer2
   Power_1stFresnel=Constant() {r=3}
   Alpha=Constant() {r=1}
  ) {materialFunction=&MF_SpherePlasma}
  n22 = MaterialFunctionCall(
   Color=@n14
   ColorFresnel=@n15
   PowerFrenelColor=@n16
   AdjustmenLayerPower=@n20.AdjustmentLayer1
   Power_1stFresnel=Constant() {r=3}
   Alpha=Constant() {r=0.33}
  ) {materialFunction=&MF_SpherePlasma}
  n23 = LinearInterpolate(B=@n22."Alpha_{", Alpha=Constant3Vector() {constant=<rgba 0.1 0.1 1 0>})
  n24 = Desaturation(None=Add(A=@n21."Alpha_{", B=Desaturation(None=@n23)))
  n25 = Multiply(A=Multiply(A=Add(A=@n19, B=@n24), B=ParticleColor().RGB), B=@n13.MultPowerPainted)
  n26 = Constant3Vector() {constant=<rgba 0 0.5 1 0>}
  n27 = If(A=@n13.IsPainted?, "A > B"=@n25, "A == B"=@n25, "A < B"=Add(A=@n18, B=LinearInterpolate(A=@n23, B=@n21."Alpha_{", Alpha=@n26))) {constB=0.5}
  n28 = ScalarParameter() {defaultValue=1, parameterName=MultiColorFinal}
  n29 = StaticBoolParameter() {parameterName=Pink?}
  MaterialFunctionCall(Emissive=StaticSwitch(True=@n12, False=Multiply(A=@n27, B=@n28), Value=@n29)) {materialFunction=&MF_FX_EmissiveAdjustments}
  Multiply(A=Constant() {r=1})
  StaticSwitchParameter()
  StaticSwitch(
   True=@n12
   False=Multiply(A=Add(A=@n21."Alpha_{", B=@n22."Alpha_{"), B=ScalarParameter() {defaultValue=2, parameterName=PowerAlpha})
   Value=@n29
  )
 ]
}
